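; Media CDN warmer actor that runs Fortnite's CDN warming on non-dedicated-server instances.
; On BeginPlay it constructs a FortMediaCDNWarmer and, if a CDN weights string is set, parses it into floats that replace the hostnames' distribution array.
; It waits a random delay of up to three minutes before starting the warmer, which staggers starts across clients.
; EndPlay stops the warmer.
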(event EventBeginPlay
  (if (not (Networking|IsDedicatedServer))
    (bind _returnvalue (Game|ConstructObjectfromClass "/Script/FortniteGame.FortMediaCDNWarmer" self))
    (Variables|Default|SetWarmer _returnvalue)
    (if (not (Utilities|String|IsEmpty (Variables|Default|GetCDNWeights)))
      (Utilities|Array|Clear (Variables|/Script/FortniteGame.FortMediaCDNHostnames|GetCDNDistribution (Variables|/Script/FortniteGame.FortMediaCDNWarmer|GetCDNHostNames (Variables|Default|GetWarmer))))
      (for _array_element (Utilities|String|ParseIntoArray (Variables|Default|GetCDNWeights))
        (Utilities|Array|Add (Variables|/Script/FortniteGame.FortMediaCDNHostnames|GetCDNDistribution (Variables|/Script/FortniteGame.FortMediaCDNWarmer|GetCDNHostNames (Variables|Default|GetWarmer))) (Utilities|String|StringToFloat _array_element))))
    (bind _output_get (Variables|Default|Settmp (Math|Random|RandomFloatinRange 0.0 1.8e+02)))
    (Development|PrintString (Utilities|String|Append "Start Simple CDN Warming in " (Utilities|String|ToString(Float) _output_get) " Seconds"))
    (Utilities|FlowControl|Delay _output_get)
    (Class|FortMediaCDNWarmer|Start (Variables|Default|GetWarmer))))

(event EventEndPlay
  (params (EndPlayReason :type EEndPlayReason%20Enum :default "Destroyed"))
  (if (not (Networking|IsDedicatedServer))
    (Class|FortMediaCDNWarmer|Stop (Variables|Default|GetWarmer))))

(event Collision|EventActorBeginOverlap
  (params (OtherActor :type Actor%20Object%20Reference)))

(event EventTick
  (params (DeltaSeconds :type Float%20%28single-precision%29)))

(fn ConstructionScript)
